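# Sound cue for the gemstone collect effect in the Season 30 Battle Pass story quest P02 Q05.
# A single wave player feeds a modulator that applies slight random pitch and volume variation on each play.
# It routes to the Fort_Quests sound class and borrows the Hidden Star ambient loop attenuation from the SupplyDrops sounds for spatial falloff.
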SoundCue /BattlePassS30_Quests/Audio/SFX/Quest_S30_Story_P02_Q05/S30_Quest_P02_Q05_Gemstone_Collect_Cue
root=Modulator_0
nodes [
 {id=Modulator_0, children [WavePlayer_0], properties {pitchMax=1.02, pitchMin=0.98, volumeMax=1.02, volumeMin=0.98}}
 {id=WavePlayer_0, properties {soundWaveAssetPtr=&/S30_Quest_P02_Q05_Gemstone_Collect}}
]
properties {
 attenuationSettings=/Game/Athena/Sounds/SupplyDrops/Fort_HiddenStar_Ambient_Loop_Att
 maxDistance=2500
 soundClassObject=&Fort_Quests
 volumeMultiplier=1
}
